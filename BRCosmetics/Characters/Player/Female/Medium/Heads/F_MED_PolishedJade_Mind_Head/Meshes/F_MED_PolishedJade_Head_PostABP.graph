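# Post-process Animation Blueprint for the female PolishedJade Mind head on the Fortnite 3L head skeleton, currently passing its input pose straight to the output.
# Each update stores a scaled, negated head-to-neck_02 rotation delta in HeadRot.
# When customization loads, the Face part name sets helmet and Mind-hair blend weights.
# Branches for counter-rotating the head, resetting facial bones and posing the jaw and lip are left unconnected to the output.
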AnimBlueprint /BRCosmetics/Characters/Player/Female/Medium/Heads/F_MED_PolishedJade_Mind_Head/Meshes/F_MED_PolishedJade_Head_PostABP
target_skeleton=/Game/Characters/Player/Common/Fortnite_Base_Head/Fortnite_Base_3L_Head_Skeleton
graphs [
 AnimGraph {nodes [
   n1 = Local Space Ref Pose()
   n2 = Get isHelmet() -> (isHelmet:double)
   n3 = "Transform (Modify) Bone - Bone: head"(
    ComponentPose=Local To Component(LocalPose=Layered blend per bone(BlendPoses_0=@n1, BlendWeights_0=@n2) {layerSetup [{branchFilters [{boneName=FACIAL_C_FacialRoot}]}]})
    Alpha=1
    Translation="0,0,0"
    Rotation=Get HeadRot() -> (HeadRot:Rotator)
    Scale="1,1,1"
   ) {rotationMode=BMM_Additive, rotationSpace=BCS_BoneSpace, alphaBoolBlend {blendOption=Linear}, boneToModify {boneName=head}}
   Component To Local(ComponentPose=@n3)
   n4 = "Transform (Modify) Bone - Bone: FACIAL_C_LowerLipRotation"(Alpha=1, Translation="0,0,0", Rotation="0,0,30", Scale="1,1,1") {
    rotationMode=BMM_Replace
    rotationSpace=BCS_BoneSpace
    alphaBoolBlend {blendOption=Linear}
    boneToModify {boneName=FACIAL_C_LowerLipRotation}
    rotation {roll=30}
   }
   n5 = "Transform (Modify) Bone - Bone: FACIAL_C_Jaw"(ComponentPose=@n4, Alpha=1, Translation="0,0,0", Rotation="0,0,30", Scale="1,1,1") {
    rotationMode=BMM_Replace
    rotationSpace=BCS_BoneSpace
    alphaBoolBlend {blendOption=Linear}
    boneToModify {boneName=FACIAL_C_Jaw}
    rotation {roll=30}
   }
   Component To Local(ComponentPose=@n5)
   Layered blend per bone(BasePose=Local Space Ref Pose(), BlendWeights_0=Get isMindHair() -> (isMindHair:double)) {
    curveBlendOption=UseBasePose
    layerSetup [{branchFilters [{blendDepth=1, boneName=FACIAL_C_FacialRoot}, {blendDepth=2, boneName=FACIAL_C_Jaw}, {blendDepth=2, boneName=FACIAL_C_LowerLipRotation}]}]
   }
   "Transform (Modify) Bone - Bone: head"(ComponentPose=Local To Component(), Alpha=1, Translation="0,0,0", Rotation="0,0,0", Scale="1,1,1") {rotationMode=BMM_Replace, rotationSpace=BCS_BoneSpace, alphaBoolBlend {blendOption=Linear}, boneToModify {boneName=head}}
   Output Pose(Result=Input Pose())
  ]}
 EventGraph {nodes [
   Event BlueprintUpdateAnimation()
   n1 = GetOwnerSkeletalMeshComponent()
   n2 = BreakRotator(InRot=BreakTransform(InTransform=GetDeltaTransformFromRefPose(self=@n1, BoneName=head, BaseName=neck_02)).Rotation)
   n3 = Get Z_Percentage() -> (Z_Percentage:double)
   n4 = Get X_Percentage() -> (X_Percentage:double)
   n5 = Get Y_Percentage() -> (Y_Percentage:double)
   > Set HeadRot(HeadRot:Rotator=MakeRotator(
     Roll=float * float!(A=@n2, B=-1, C:double=@n4) -> (ReturnValue) {is=title K2Node_PromotableOperator}
     Pitch=float * float(A=@n2.Pitch, B=-1, C=@n5)
     Yaw=float * float(A=@n3, B=-1, C:float=@n2.Yaw)
    )) -> (Output_Get:Rotator)
   n6 = Event OnCharacterCustomizationLoaded()
   n7 = GetDisplayName(Object=Get Array Item with EFort Custom Part Type(Array:Array of Custom Character Part Object References=@n6.CustomParts, PartType=Face) -> (Output:Custom Character Part Object Reference))
   n8 = Contains(SearchIn=@n7, Substring=MindHelmet)
   > Set isHelmet(isHelmet:double="To Float (Boolean)"(InBool=@n8)) -> (Output_Get:double)
   n9 = NOT Boolean(A=@n8)
   n10 = Contains(SearchIn=@n7, Substring=Mind)
   > Set isMindHair(isMindHair:double="To Float (Boolean)"(InBool=AND Boolean(A=@n9, B=@n10))) -> (Output_Get:double)
   TryGetPawnOwner()
  ]}
]
variables [
 HeadRot {type=Rotator}
 isHelmet {type=number, default=1}
 isMindHair {type=number}
 X_Percentage {type=number, default=0.9}
 Y_Percentage {type=number, default=0.8}
 Z_Percentage {type=number, default=0.8}
]
